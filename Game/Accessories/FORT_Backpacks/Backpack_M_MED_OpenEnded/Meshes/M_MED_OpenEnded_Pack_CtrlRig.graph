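# Control Rig for the Backpack_M_MED_OpenEnded pack, adding procedural secondary motion to its tentacle and brain bones.
# Two chain harmonics units, with pendulum, reach, wave and debug drawing enabled, drive the left and second top tentacle chains from mid-chain bones.
# Three item harmonics units follow in sequence, adding noisy sine-eased waves to both brain bones and the first right tentacle's base.
# The back, first top and second right tentacles have no driving unit.
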ControlRigBlueprint /Game/Accessories/FORT_Backpacks/Backpack_M_MED_OpenEnded/Meshes/M_MED_OpenEnded_Pack_CtrlRig
graphs [
 RigVMFunctionLibrary {links [], nodes []}
 RigVMModel {nodes [
   /Script/ControlRig.RigUnit_ChainHarmonicsPerItem(
    ChainRoot=(Type=Bone,Name="dyn_tentacle_l_02")
    DrawWorldOffset=(Rotation=(X=0,Y=0,Z=0,W=1),Translation=(X=0,Y=0,Z=0),Scale3D=(X=1,Y=1,Z=1))
    ExecutePin=()
    Pendulum=(bEnabled=True,PendulumStiffness=2,PendulumGravity=(X=0,Y=0,Z=0),PendulumBlend=0.75,PendulumDrag=0.98,PendulumMaximum=0.1,UnwindAxis=(X=0,Y=1,Z=0),UnwindMinimum=0.2,UnwindMaximum=0.05)
    Reach=(bEnabled=True,ReachTarget=(X=1,Y=1,Z=1),ReachAxis=(X=0,Y=0,Z=0),ReachMinimum=0,ReachMaximum=0)
    Speed=(X=0.5,Y=0.3,Z=0.5)
    Wave=(bEnabled=True,WaveFrequency=(X=1,Y=0,Z=1),WaveAmplitude=(X=0.3,Y=0.3,Z=0.3),WaveOffset=(X=0,Y=0,Z=0),WaveNoise=(X=0.8,Y=0,Z=0.8),WaveMinimum=-0.5,WaveMaximum=-1)
    WaveCurve="(EditorCurveData=(Keys=((InterpMode=RCIM_Cubic,Time=0.1619,Value=0.3394),(InterpMode=RCIM_Cubic,Time=0.3504,Value=0.3613,ArriveTangent=0.2483,LeaveTangent=0.2483),(InterpMode=RCIM_Cubic,Time=0.4991,Value=0.4231,ArriveTangent=0.5281,LeaveTangent=0.5281),(In ...(truncated)"
    bDrawDebug=True
   ) {is=struct RigVMUnitNode}
   > &RigUnit_ChainHarmonicsPerItem(
    ChainRoot=(Type=Bone,Name="dyn_tentacle_top02_03")
    DrawWorldOffset=(Rotation=(X=0,Y=0,Z=0,W=1),Translation=(X=0,Y=0,Z=0),Scale3D=(X=1,Y=1,Z=1))
    Pendulum=(bEnabled=True,PendulumStiffness=2,PendulumGravity=(X=0,Y=0,Z=0),PendulumBlend=0.75,PendulumDrag=0.98,PendulumMaximum=0.1,UnwindAxis=(X=0,Y=1,Z=0),UnwindMinimum=0.2,UnwindMaximum=0.05)
    Reach=(bEnabled=True,ReachTarget=(X=0,Y=0,Z=0),ReachAxis=(X=0,Y=0,Z=0),ReachMinimum=0,ReachMaximum=0)
    Speed=(X=1.1,Y=1.1,Z=0.8)
    Wave=(bEnabled=True,WaveFrequency=(X=0,Y=0,Z=0),WaveAmplitude=(X=0.5,Y=0.1,Z=0),WaveOffset=(X=1,Y=1,Z=1),WaveNoise=(X=1,Y=0,Z=1),WaveMinimum=-0.5,WaveMaximum=-1)
    WaveCurve="(EditorCurveData=(Keys=((Time=0.1619,Value=0.3394),(InterpMode=RCIM_Cubic,Time=0.3504,Value=0.3613,ArriveTangent=0.2483,LeaveTangent=0.2483),(InterpMode=RCIM_Cubic,Time=0.4991,Value=0.4231,ArriveTangent=0.5281,LeaveTangent=0.5281),(InterpMode=RCIM_Cubic,Ti ...(truncated)"
    bDrawDebug=True
   )
   > /Script/ControlRig.RigUnit_ItemHarmonics(
    RotationOrder=YZX
    Targets=((Item=(Type=Bone,Name="dyn_brain_01"),Ratio=0))
    WaveAmplitude=(X=2,Y=2,Z=2)
    WaveEase=SineEaseIn
    WaveFrequency=(X=0,Y=0,Z=0)
    WaveMaximum=1
    WaveMinimum=0.5
    WaveNoise=(X=1.3,Y=1.3,Z=1.3)
    WaveOffset=(X=1,Y=1,Z=1)
    WaveSpeed=(X=1.3,Y=1.3,Z=1.3)
   ) {is=struct RigVMUnitNode}
   > &RigUnit_ItemHarmonics(
    RotationOrder=YZX
    Targets=((Item=(Type=Bone,Name="dyn_brain_02"),Ratio=0))
    WaveAmplitude=(X=1.5,Y=1.5,Z=1.5)
    WaveEase=SineEaseIn
    WaveFrequency=(X=0,Y=0,Z=0)
    WaveMaximum=1
    WaveMinimum=0.5
    WaveNoise=(X=1.2,Y=1.2,Z=1.2)
    WaveOffset=(X=0,Y=0,Z=0)
    WaveSpeed=(X=1.2,Y=1.2,Z=1.2)
   )
   > &RigUnit_ItemHarmonics(
    RotationOrder=YZX
    Targets=((Item=(Type=Bone,Name="dyn_tentacle_r01_01"),Ratio=0))
    WaveAmplitude=(X=0.7,Y=0.7,Z=0.7)
    WaveEase=SineEaseIn
    WaveFrequency=(X=0,Y=0,Z=0)
    WaveMaximum=1
    WaveMinimum=0.5
    WaveNoise=(X=5,Y=5,Z=5)
    WaveOffset=(X=0.5,Y=0.5,Z=0.5)
    WaveSpeed=(X=0.5,Y=0.5,Z=0.5)
   )
   RigUnit_BeginExecution = &RigUnit_BeginExecution()
  ]}
]
hierarchy [
 root {loc=<0 0 0>, rot=<0 0 0 1>}
 pack {parent=root, loc=<0 -0.7316 0>, rot=<0 0 0 1>}
 dyn_tentacle_top02_01 {parent=pack, loc=<-12.2055 -6.3145 6.4862>, rot=<0.2489 0.4784 -0.0335 -0.8415>}
 dyn_tentacle_top02_02 {parent=dyn_tentacle_top02_01, loc=<5.8667 0.061 -1.045>, rot=<-0.093 -0.4069 0.1676 -0.8931>}
 dyn_tentacle_top02_03 {parent=dyn_tentacle_top02_02, loc=<4.8639 0.0173 -0.3273>, rot=<-0.0814 -0.3248 0.1595 -0.9287>}
 dyn_tentacle_top02_04 {parent=dyn_tentacle_top02_03, loc=<3.5355 0.2782 -0.952>, rot=<-0.1946 -0.2604 -0.0052 -0.9457>}
 dyn_tentacle_top02_05 {parent=dyn_tentacle_top02_04, loc=<4.4154 0.4667 -0.1362>, rot=<0.4819 0.2131 -0.2144 -0.8224>}
 dyn_brain_02 {parent=pack, loc=<-13.4218 -9.5649 1.7364>, rot=<0.002 0.0082 1 -0.0004>}
 dyn_brain_01 {parent=pack, loc=<5.3961 -6.3983 10.9932>, rot=<0 0.7071 0 -0.7071>}
 dyn_tentacle_l_01 {parent=pack, loc=<8.1157 -9.3346 7.4613>, rot=<-0.1682 -0.458 0.313 -0.8148>}
 dyn_tentacle_l_02 {parent=dyn_tentacle_l_01, loc=<8.3677 0.1076 -0.4098>, rot=<-0.0149 -0.0858 0.007 -0.9962>}
 dyn_tentacle_l_03 {parent=dyn_tentacle_l_02, loc=<6.8372 -0.8606 -2.6504>, rot=<-0.5778 0.0707 0.8046 -0.1172>}
 dyn_tentacle_l_04 {parent=dyn_tentacle_l_03, loc=<6.7941 -0.104 -0.5552>, rot=<-0.0501 -0.3608 0.1663 -0.9163>}
 dyn_tentacle_l_05 {parent=dyn_tentacle_l_04, loc=<4.365 -1.016 -3.0914>, rot=<0.1601 -0.2563 -0.2488 -0.9202>}
 dyn_tentacle_top_01 {parent=pack, loc=<-5.5014 -6.3331 10.5978>, rot=<-0.0937 -0.2029 0.8821 -0.4146>}
 dyn_tentacle_top_02 {parent=dyn_tentacle_top_01, loc=<4.445 0.0727 -1.2745>, rot=<0.0566 -0.2992 -0.0444 -0.9515>}
 dyn_tentacle_r01_01 {parent=pack, loc=<-11.5531 -12.9785 -0.8856>, rot=<-0.6543 -0.0567 0.4193 -0.6268>}
 dyn_tentacle_r01_02 {parent=dyn_tentacle_r01_01, loc=<8.1974 0.7673 -1.3206>, rot=<0.0477 -0.2114 -0.2212 -0.9509>}
 dyn_tentacle_r01_03 {parent=dyn_tentacle_r01_02, loc=<7.1044 0.307 -2.4707>, rot=<0.6781 -0.2168 -0.4029 -0.5752>}
 dyn_tentacle_r01_04 {parent=dyn_tentacle_r01_03, loc=<5.5966 2.0862 0.3124>, rot=<-0.0448 0.0395 -0.4791 -0.8757>}
 dyn_tentacle_r01_05 {parent=dyn_tentacle_r01_04, loc=<3.9229 0.2006 -0.0981>, rot=<-0.008 -0.0539 -0.2967 -0.9534>}
 dyn_tentacle_back_01 {parent=pack, loc=<10.1919 -4.6686 10.0107>, rot=<0.1749 -0.4326 -0.3316 -0.82>}
 dyn_tentacle_back_02 {parent=dyn_tentacle_back_01, loc=<5.4079 1.6855 -1.8707>, rot=<0.1166 -0.3759 -0.4283 -0.8134>}
 dyn_tentacle_back_03 {parent=dyn_tentacle_back_02, loc=<5.8199 2.11 -1.5647>, rot=<0.0561 -0.1536 -0.3323 -0.9289>}
 dyn_tentacle_back_04 {parent=dyn_tentacle_back_03, loc=<6.411 -0.5565 0.7456>, rot=<0.0531 0.068 0.2131 -0.9732>}
 dyn_tentacle_back_05 {parent=dyn_tentacle_back_04, loc=<4.432 -1.5685 0.5494>, rot=<0.0045 0.0473 0.4621 -0.8856>}
 dyn_tentacle_r02_01 {parent=pack, loc=<-12.1399 -8.8585 -3.9312>, rot=<-0.5962 -0.1204 0.7574 0.2374>}
 dyn_tentacle_r02_02 {parent=dyn_tentacle_r02_01, loc=<3.9931 -0.5802 -3.1813>, rot=<-0.6021 0.1008 0.6255 -0.4858>}
 dyn_tentacle_r02_03 {parent=dyn_tentacle_r02_02, loc=<6.7564 0.6446 0.6881>, rot=<-0.0007 0.2436 0.0012 -0.9699>}
 dyn_tentacle_r02_04 {parent=dyn_tentacle_r02_03, loc=<3.9814 -0.9095 0.7563>, rot=<0.0513 -0.0324 0.3085 -0.9493>}
 dyn_tentacle_r02_05 {parent=dyn_tentacle_r02_04, loc=<3.5708 -0.4243 -0.2919>, rot=<0.0859 -0.2953 0.0532 -0.9501>}
]
