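# Physics asset for the OpenEnded medium male backpack, giving its brain and tentacle bones capsule bodies for physics-driven secondary motion.
# Two brain bodies and the first link of six tentacle chains (back, left, two right, two top) constrain directly to the pack body.
# Later links constrain each to the previous one, forming five-link chains except one two-link top chain.
# Angular limits are very tight for slight wobble, while pack-to-root rotation is free.
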PhysicsAsset /Game/Accessories/FORT_Backpacks/Backpack_M_MED_OpenEnded/Meshes/M_MED_OpenEnded_Pack_Physics
bodies {
 dyn_brain_01 [{
   shapeType=Capsule
   length=4.9169
   radius=4.4634
   shapeName=dyn_brain_01_capsule
   center=<1.4597 -0.6787 0.3119>
   rotation=<rot 27.0404 -90 14.8347>
  }]
 dyn_brain_02 [{
   shapeType=Capsule
   length=4.9098
   radius=4.4926
   shapeName=dyn_brain_02_capsule
   center=<0.8824 0.0253 0.1144>
   rotation=<rot 0 -3.1586 -69.4495>
  }]
 dyn_tentacle_back_01 [{
   shapeType=Capsule
   length=4.1358
   radius=1
   shapeName=dyn_tentacle_back_01_capsule
   center=<2.0279 0.4051 -0.7711>
   rotation=<rot 0 -70.4055 -81.8415>
  }]
 dyn_tentacle_back_02 [{
   shapeType=Capsule
   length=3.515
   radius=1
   shapeName=dyn_tentacle_back_02_capsule
   center=<2.9016 0.3279 -0.5979>
   rotation=<rot 0 -65.483 -76.3381>
  }]
 dyn_tentacle_back_03 [{
   shapeType=Capsule
   length=3.4058
   radius=1
   shapeName=dyn_tentacle_back_03_capsule
   center=<3.3601 -0.1016 0.2568>
   rotation=<rot 0 89.9954 -86.3815>
  }]
 dyn_tentacle_back_04 [{
   shapeType=Capsule
   length=1.668
   radius=1
   shapeName=dyn_tentacle_back_04_capsule
   center=<1.5812 -0.0921 0.1721>
   rotation=<rot 0 87.2849 -87.3964>
  }]
 dyn_tentacle_back_05 [{
   shapeType=Capsule
   length=1.9526
   radius=1
   shapeName=dyn_tentacle_back_05_capsule
   center=<0.6798 -0.257 0.0708>
   rotation=<rot 0 83.8948 -85.6504>
  }]
 dyn_tentacle_l_01 [{
   shapeType=Capsule
   length=4.5777
   radius=1
   shapeName=dyn_tentacle_l_01_capsule
   center=<1.8284 0.0142 -0.4519>
   rotation=<rot 0 -96.5632 -85.0056>
  }]
 dyn_tentacle_l_02 [{
   shapeType=Capsule
   length=4.5121
   radius=1
   shapeName=dyn_tentacle_l_02_capsule
   center=<3.6822 -0.4355 -1.1623>
   rotation=<rot 0 -95.6923 -75.4612>
  }]
 dyn_tentacle_l_03 [{
   shapeType=Capsule
   length=3.9892
   radius=1
   shapeName=dyn_tentacle_l_03_capsule
   center=<3.7542 0.0599 -0.4329>
   rotation=<rot 0 89.1318 -86.7081>
  }]
 dyn_tentacle_l_04 [{
   shapeType=Capsule
   length=2.5878
   radius=1
   shapeName=dyn_tentacle_l_04_capsule
   center=<2.4285 -0.4146 -1.101>
   rotation=<rot 0 -111.0865 -58.4981>
  }]
 dyn_tentacle_l_05 [{
   shapeType=Capsule
   length=1.4248
   radius=1
   shapeName=dyn_tentacle_l_05_capsule
   center=<1.2249 0.0234 0.0007>
   rotation=<rot 0 95.6452 -74.3533>
  }]
 dyn_tentacle_r01_01 [{
   shapeType=Capsule
   length=5.1211
   radius=1
   shapeName=dyn_tentacle_r01_01_capsule
   center=<4.5642 0.2748 -0.1901>
   rotation=<rot 0 -82.8279 -79.915>
  }]
 dyn_tentacle_r01_02 [{
   shapeType=Capsule
   length=4.4866
   radius=1
   shapeName=dyn_tentacle_r01_02_capsule
   center=<3.693 0.0716 -0.6729>
   rotation=<rot 0 -81.6272 -70.717>
  }]
 dyn_tentacle_r01_03 [{
   shapeType=Capsule
   length=2.0076
   radius=1
   shapeName=dyn_tentacle_r01_03_capsule
   center=<1.8986 0.107 0.0518>
   rotation=<rot 0 103.3105 -83.7044>
  }]
 dyn_tentacle_r01_04 [{
   shapeType=Capsule
   length=4.109
   radius=1
   shapeName=dyn_tentacle_r01_04_capsule
   center=<2.3962 0.5633 0.0361>
   rotation=<rot 0 -78.1673 -89.5404>
  }]
 dyn_tentacle_r01_05 [{shapeType=Capsule, length=0.505, radius=1, shapeName=dyn_tentacle_r01_05_capsule, rotation=<rot 0 -90 0>}]
 dyn_tentacle_r02_01 [{
   shapeType=Capsule
   length=3.5583
   radius=1
   shapeName=dyn_tentacle_r02_01_capsule
   center=<0.704 0.2573 -0.4679>
   rotation=<rot 0 -116.3241 -60.9241>
  }]
 dyn_tentacle_r02_02 [{
   shapeType=Capsule
   length=2.9113
   radius=1
   shapeName=dyn_tentacle_r02_02_capsule
   center=<2.6773 0.2855 -0.1887>
   rotation=<rot 0 -85.9984 -89.0051>
  }]
 dyn_tentacle_r02_03 [{
   shapeType=Capsule
   length=1.8129
   radius=1
   shapeName=dyn_tentacle_r02_03_capsule
   center=<1.0987 0.0188 0.136>
   rotation=<rot 0 86.6092 -78.3544>
  }]
 dyn_tentacle_r02_04 [{
   shapeType=Capsule
   length=2.1845
   radius=1
   shapeName=dyn_tentacle_r02_04_capsule
   center=<1.0994 -0.201 0.0092>
   rotation=<rot 0 85.7946 -89.9868>
  }]
 dyn_tentacle_r02_05 [{
   shapeType=Capsule
   length=1.7357
   radius=1
   shapeName=dyn_tentacle_r02_05_capsule
   center=<1.3901 0.0497 -0.3459>
   rotation=<rot 0 -88.0377 -72.0855>
  }]
 dyn_tentacle_top02_01 [{
   shapeType=Capsule
   length=4.8469
   radius=1
   shapeName=dyn_tentacle_top02_01_capsule
   center=<1.9508 0.2479 -0.5105>
   rotation=<rot 0 91.3709 -89.2037>
  }]
 dyn_tentacle_top02_02 [{
   shapeType=Capsule
   length=1.8519
   radius=1
   shapeName=dyn_tentacle_top02_02_capsule
   center=<2.0692 0.2799 0.0367>
   rotation=<rot 0 -87.7796 -88.2158>
  }]
 dyn_tentacle_top02_03 [{
   shapeType=Capsule
   length=2.482
   radius=1
   shapeName=dyn_tentacle_top02_03_capsule
   center=<1.3165 -0.0328 -0.3073>
   rotation=<rot 0 -86.8408 -76.6915>
  }]
 dyn_tentacle_top02_04 [{
   shapeType=Capsule
   length=3.6708
   radius=1
   shapeName=dyn_tentacle_top02_04_capsule
   center=<3.7482 0.9152 0.1737>
   rotation=<rot 0 115.2735 -84.3771>
  }]
 dyn_tentacle_top02_05 [{shapeType=Capsule, length=0.505, radius=1, shapeName=dyn_tentacle_top02_05_capsule, rotation=<rot 0 -90 0>}]
 dyn_tentacle_top_01 [{
   shapeType=Capsule
   length=4.0202
   radius=1
   shapeName=dyn_tentacle_top_01_capsule
   center=<3.7228 0.3286 -1.0911>
   rotation=<rot 0 -87.3413 -61.1043>
  }]
 dyn_tentacle_top_02 [{
   shapeType=Capsule
   length=1.5303
   radius=1
   shapeName=dyn_tentacle_top_02_capsule
   center=<2.9022 0.0629 -0.0898>
   rotation=<rot 0 61.2928 -87.5759>
  }]
 pack [
  {shapeType=Capsule, length=1, radius=1}
  {
   shapeType=Capsule
   length=7.3221
   radius=5.6486
   shapeName=pack_capsule
   center=<-0.0057 -7.688 0.8483>
   rotation=<rot 89.5671 -179.9066 -179.9121>
  }
 ]
 root [{shapeType=Capsule, length=0.505, radius=0.505, shapeName=root_capsule, rotation=<rot 0 -90 0>}]
}
constraints [
 {
  bone1Name=pack
  bone2Name=root
  swing1LimitDegrees=45
  swing1Motion=Free
  swing2LimitDegrees=45
  swing2Motion=Free
  twistLimitDegrees=45
  twistMotion=Free
 }
 {bone1Name=dyn_brain_01, bone2Name=pack, swing1LimitDegrees=3, swing2LimitDegrees=3, twistLimitDegrees=1}
 {bone1Name=dyn_brain_02, bone2Name=pack, swing1LimitDegrees=4, swing2LimitDegrees=4, twistLimitDegrees=1}
 {bone1Name=dyn_tentacle_back_01, bone2Name=pack, swing1LimitDegrees=1, swing2LimitDegrees=1, twistLimitDegrees=1}
 {bone1Name=dyn_tentacle_back_02, bone2Name=dyn_tentacle_back_01, swing1LimitDegrees=2, swing2LimitDegrees=2, twistLimitDegrees=2}
 {bone1Name=dyn_tentacle_back_03, bone2Name=dyn_tentacle_back_02, swing1LimitDegrees=2, swing2LimitDegrees=2, twistLimitDegrees=1}
 {bone1Name=dyn_tentacle_back_04, bone2Name=dyn_tentacle_back_03, swing1LimitDegrees=3, swing2LimitDegrees=3, twistLimitDegrees=1}
 {bone1Name=dyn_tentacle_back_05, bone2Name=dyn_tentacle_back_04, swing1LimitDegrees=2, swing2LimitDegrees=2, twistLimitDegrees=1}
 {bone1Name=dyn_tentacle_l_01, bone2Name=pack, swing1LimitDegrees=1, swing2LimitDegrees=1, twistLimitDegrees=1}
 {bone1Name=dyn_tentacle_l_02, bone2Name=dyn_tentacle_l_01, swing1LimitDegrees=2, swing2LimitDegrees=2, twistLimitDegrees=2}
 {bone1Name=dyn_tentacle_l_03, bone2Name=dyn_tentacle_l_02, swing1LimitDegrees=2, swing2LimitDegrees=2, twistLimitDegrees=1}
 {bone1Name=dyn_tentacle_l_04, bone2Name=dyn_tentacle_l_03, swing1LimitDegrees=3, swing2LimitDegrees=3, twistLimitDegrees=1}
 {bone1Name=dyn_tentacle_l_05, bone2Name=dyn_tentacle_l_04, swing1LimitDegrees=2, swing2LimitDegrees=2, twistLimitDegrees=1}
 {bone1Name=dyn_tentacle_r01_01, bone2Name=pack, swing1LimitDegrees=1, swing2LimitDegrees=1, twistLimitDegrees=1}
 {bone1Name=dyn_tentacle_r01_02, bone2Name=dyn_tentacle_r01_01, swing1LimitDegrees=2, swing2LimitDegrees=2, twistLimitDegrees=2}
 {bone1Name=dyn_tentacle_r01_03, bone2Name=dyn_tentacle_r01_02, swing1LimitDegrees=2, swing2LimitDegrees=2, twistLimitDegrees=1}
 {bone1Name=dyn_tentacle_r01_04, bone2Name=dyn_tentacle_r01_03, swing1LimitDegrees=3, swing2LimitDegrees=3, twistLimitDegrees=1}
 {bone1Name=dyn_tentacle_r01_05, bone2Name=dyn_tentacle_r01_04, swing1LimitDegrees=2, swing2LimitDegrees=2, twistLimitDegrees=1}
 {bone1Name=dyn_tentacle_r02_01, bone2Name=pack, swing1LimitDegrees=1, swing2LimitDegrees=1, twistLimitDegrees=1}
 {bone1Name=dyn_tentacle_r02_02, bone2Name=dyn_tentacle_r02_01, swing1LimitDegrees=2, swing2LimitDegrees=2, twistLimitDegrees=2}
 {bone1Name=dyn_tentacle_r02_03, bone2Name=dyn_tentacle_r02_02, swing1LimitDegrees=2, swing2LimitDegrees=2, twistLimitDegrees=1}
 {bone1Name=dyn_tentacle_r02_04, bone2Name=dyn_tentacle_r02_03, swing1LimitDegrees=3, swing2LimitDegrees=3, twistLimitDegrees=1}
 {bone1Name=dyn_tentacle_r02_05, bone2Name=dyn_tentacle_r02_04, swing1LimitDegrees=2, swing2LimitDegrees=2, twistLimitDegrees=1}
 {bone1Name=dyn_tentacle_top02_01, bone2Name=pack, swing1LimitDegrees=1, swing2LimitDegrees=1, twistLimitDegrees=1}
 {bone1Name=dyn_tentacle_top02_02, bone2Name=dyn_tentacle_top02_01, swing1LimitDegrees=2, swing2LimitDegrees=2, twistLimitDegrees=2}
 {bone1Name=dyn_tentacle_top02_03, bone2Name=dyn_tentacle_top02_02, swing1LimitDegrees=2, swing2LimitDegrees=2, twistLimitDegrees=1}
 {bone1Name=dyn_tentacle_top02_04, bone2Name=dyn_tentacle_top02_03, swing1LimitDegrees=3, swing2LimitDegrees=3, twistLimitDegrees=1}
 {bone1Name=dyn_tentacle_top02_05, bone2Name=dyn_tentacle_top02_04, swing1LimitDegrees=2, swing2LimitDegrees=2, twistLimitDegrees=1}
 {bone1Name=dyn_tentacle_top_01, bone2Name=pack, swing1LimitDegrees=1, swing2LimitDegrees=1, twistLimitDegrees=1}
 {bone1Name=dyn_tentacle_top_02, bone2Name=dyn_tentacle_top_01, swing1LimitDegrees=2, swing2LimitDegrees=2, twistLimitDegrees=2}
]
